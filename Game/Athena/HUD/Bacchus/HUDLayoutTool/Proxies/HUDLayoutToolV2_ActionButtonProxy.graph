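# Widget Blueprint proxy for an action button in the HUD Layout Tool V2, built on the native HUDLayoutToolV2_ActionButtonProxy class with an empty event graph.
# A hitbox SizeBox wraps a separately sized visual SizeBox, whose ScaleBox scales a fixed-size background image and semi-transparent icon, defaulting to a Confirm graphic.
# Collapsed out-of-bounds and selection highlight images overlay the hitbox with negative padding, so they extend slightly past its edges.
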WidgetBlueprint /Game/Athena/HUD/Bacchus/HUDLayoutTool/Proxies/HUDLayoutToolV2_ActionButtonProxy
graphs [EventGraph {links [], nodes []}]
variables []
widgets {
 info {
  inheritedWidgetCount=8
  parentClass=/Script/FortniteUI.HUDLayoutToolV2_ActionButtonProxy
  rootWidgetClass=&SizeBox
  widgetCount=13
 }
 widgets [
  SizeBox_HitboxSize {
   widgetClassPath=&SizeBox
   bInherited=true
   widget=SizeBox_HitboxSize
   properties {bOverride_HeightOverride=true, bOverride_WidthOverride=true, heightOverride=200, visibility=Visible, widthOverride=200}
  }
  Overlay_2 {widgetClassPath=&Overlay, parent=SizeBox_HitboxSize, widget=Overlay_2, slot {class=SizeBoxSlot}}
  SizeBox_VisualSize {
   widgetClassPath=&SizeBox
   bInherited=true
   parent=Overlay_2
   widget=SizeBox_VisualSize
   properties {bOverride_HeightOverride=true, bOverride_WidthOverride=true, heightOverride=200, widthOverride=200}
   slot {class=OverlaySlot, properties {horizontalAlignment=HAlign_Center, verticalAlignment=VAlign_Center}}
  }
  ScaleBox_0 {widgetClassPath=&ScaleBox, parent=SizeBox_VisualSize, widget=ScaleBox_0, slot {class=SizeBoxSlot}}
  SizeBox_1 {
   widgetClassPath=&SizeBox
   parent=ScaleBox_0
   widget=SizeBox_1
   properties {bOverride_HeightOverride=true, bOverride_WidthOverride=true, heightOverride=128, widthOverride=128}
   slot {class=ScaleBoxSlot}
  }
  Overlay {widgetClassPath=&Overlay, parent=SizeBox_1, widget=Overlay, slot {class=SizeBoxSlot}}
  Image_Background {
   widgetClassPath=&Image
   bInherited=true
   bIsVariable=true
   parent=Overlay
   widget=Image_Background
   properties {brush {
     bIsDynamicallyLoaded=false
     drawAs=Image
     imageType=FullColor
     mirroring=NoMirror
     resourceName=None
     resourceObject=/Game/Athena/HUD/Bacchus/Material/MI_UI_ButtonBG_Light
     tiling=NoTile
     imageSize=<128 128>
     margin=<ltrb 0 0 0 0>
     outlineSettings {
      bUseBrushTransparency=false
      roundingType=HalfHeightRadius
      width=0
      color {colorUseRule=UseColor_Specified, specifiedColor=<rgba 0 0 0 0>}
      cornerRadii=<0 0 0 0>
     }
     tintColor {colorUseRule=UseColor_Specified, specifiedColor=<rgba 1 1 1 1>}
     uVRegion {bIsValid=false, max=<0 0>, min=<0 0>}
    }}
   slot {class=OverlaySlot, properties {horizontalAlignment=HAlign_Fill, verticalAlignment=VAlign_Fill}}
  }
  SizeBox_0 {
   widgetClassPath=&SizeBox
   parent=Overlay
   widget=SizeBox_0
   properties {
    bOverride_MaxDesiredHeight=true
    bOverride_MaxDesiredWidth=true
    bOverride_MinDesiredHeight=true
    bOverride_MinDesiredWidth=true
    heightOverride=128
    maxDesiredHeight=128
    maxDesiredWidth=128
    minDesiredHeight=60
    minDesiredWidth=60
    widthOverride=128
   }
   slot {class=OverlaySlot, properties {horizontalAlignment=HAlign_Center, verticalAlignment=VAlign_Center}}
  }
  Image_Icon {
   widgetClassPath=&Image
   bInherited=true
   bIsVariable=true
   parent=SizeBox_0
   widget=Image_Icon
   properties {
    brush {
     bIsDynamicallyLoaded=false
     drawAs=Image
     imageType=FullColor
     mirroring=NoMirror
     resourceName=None
     resourceObject=/Game/Athena/HUD/Art/BacchusHudIcons/Confirm_Normal_L
     tiling=NoTile
     imageSize=<72 72>
     margin=<ltrb 0 0 0 0>
     outlineSettings {
      bUseBrushTransparency=false
      roundingType=HalfHeightRadius
      width=0
      color {colorUseRule=UseColor_Specified, specifiedColor=<rgba 0 0 0 0>}
      cornerRadii=<0 0 0 0>
     }
     tintColor {colorUseRule=UseColor_Specified, specifiedColor=<rgba 1 1 1 1>}
     uVRegion {bIsValid=false, max=<0 0>, min=<0 0>}
    }
    colorAndOpacity=<rgba 1 1 1 0.6>
   }
   slot {class=SizeBoxSlot}
  }
  Image_OutOfBoundsIndicator {
   widgetClassPath=&Image
   bInherited=true
   bIsVariable=true
   parent=Overlay_2
   widget=Image_OutOfBoundsIndicator
   properties {
    visibility=Collapsed
    brush {
     bIsDynamicallyLoaded=false
     drawAs=Image
     imageType=NoImage
     mirroring=NoMirror
     resourceName=None
     resourceObject=/Game/Athena/HUD/Bacchus/HUDLayoutTool/HLT_Icons/MI_HLT_OutOfBoundsIndicator
     tiling=NoTile
     imageSize=<32 32>
     margin=<ltrb 0 0 0 0>
     outlineSettings {
      bUseBrushTransparency=false
      roundingType=HalfHeightRadius
      width=0
      color {colorUseRule=UseColor_Specified, specifiedColor=<rgba 0 0 0 0>}
      cornerRadii=<0 0 0 0>
     }
     tintColor {colorUseRule=UseColor_Specified, specifiedColor=<rgba 1 1 1 1>}
     uVRegion {bIsValid=false, max=<0 0>, min=<0 0>}
    }
   }
   slot {class=OverlaySlot, properties {horizontalAlignment=HAlign_Fill, verticalAlignment=VAlign_Fill, padding=<ltrb -2 -2 -2 -2>}}
  }
  Image_SelectionHighlight {
   widgetClassPath=&Image
   bInherited=true
   bIsVariable=true
   parent=Overlay_2
   widget=Image_SelectionHighlight
   properties {
    visibility=Collapsed
    brush {
     bIsDynamicallyLoaded=false
     drawAs=Box
     imageType=NoImage
     mirroring=NoMirror
     resourceName=None
     resourceObject=/Game/Athena/HUD/Bacchus/HUDLayoutTool/HLT_Icons/M_UI_Square_Inst
     tiling=NoTile
     imageSize=<32 32>
     margin=<ltrb 0.25 0.25 0.25 0.25>
     outlineSettings {
      bUseBrushTransparency=false
      roundingType=HalfHeightRadius
      width=0
      color {colorUseRule=UseColor_Specified, specifiedColor=<rgba 0 0 0 0>}
      cornerRadii=<0 0 0 0>
     }
     tintColor {colorUseRule=UseColor_Specified, specifiedColor=<rgba 1 1 1 1>}
     uVRegion {bIsValid=false, max=<0 0>, min=<0 0>}
    }
    colorAndOpacity=<rgba 0.9301 1 0.0097 1>
   }
   slot {class=OverlaySlot, properties {horizontalAlignment=HAlign_Fill, verticalAlignment=VAlign_Fill, padding=<ltrb -4 -4 -4 -4>}}
  }
  ScaleBox_HUDWidgetScale {widgetClassPath=&ScaleBox, bInherited=true, widget=null}
  TutorialHighlight {widgetClassPath=&TutorialHighlight}
 ]
}
