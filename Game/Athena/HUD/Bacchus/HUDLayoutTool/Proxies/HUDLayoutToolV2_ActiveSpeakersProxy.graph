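# Widget Blueprint proxy that stands in for the active voice speakers element in the mobile HUD Layout Tool, based on FortMobileHUDElementProxy.
# A scaled horizontal row shows three identical mock speaker entries, each a placeholder image with static user name and channel name text.
# Inherited out-of-bounds and selection highlight images overlay the preview and start collapsed.
# It has no graph logic or variables of its own, only layout.
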WidgetBlueprint /Game/Athena/HUD/Bacchus/HUDLayoutTool/Proxies/HUDLayoutToolV2_ActiveSpeakersProxy
graphs [EventGraph {links [], nodes []}]
variables []
widgets {
 info {inheritedWidgetCount=5, parentClass=/Script/FortniteUI.FortMobileHUDElementProxy, rootWidgetClass=&ScaleBox, widgetCount=28}
 widgets [
  ScaleBox_HUDWidgetScale {widgetClassPath=&ScaleBox, bInherited=true, widget=ScaleBox_HUDWidgetScale, properties {stretch=UserSpecified}}
  Overlay_0 {
   widgetClassPath=&Overlay
   parent=ScaleBox_HUDWidgetScale
   widget=Overlay_0
   slot {class=ScaleBoxSlot, properties {horizontalAlignment=HAlign_Fill, verticalAlignment=VAlign_Fill}}
  }
  ScaleBox_0 {
   widgetClassPath=&ScaleBox
   parent=Overlay_0
   widget=ScaleBox_0
   properties {stretch=UserSpecified, userSpecifiedScale=1.05}
   slot {class=OverlaySlot, properties {horizontalAlignment=HAlign_Fill, verticalAlignment=VAlign_Fill}}
  }
  VerticalBox_0 {widgetClassPath=&VerticalBox, parent=ScaleBox_0, widget=VerticalBox_0, properties {visibility=Visible}, slot {class=ScaleBoxSlot}}
  SizeBox_HeightReservation {
   widgetClassPath=&SizeBox
   parent=VerticalBox_0
   widget=SizeBox_HeightReservation
   properties {bOverride_MinDesiredHeight=true, heightOverride=32, minDesiredHeight=52}
   slot {class=VerticalBoxSlot, properties {padding=<ltrb 0 5 0 0>}}
  }
  HorizontalBox_0 {widgetClassPath=&HorizontalBox, parent=SizeBox_HeightReservation, widget=HorizontalBox_0, slot {class=SizeBoxSlot}}
  Overlay_Speaker1 {
   widgetClassPath=&Overlay
   parent=HorizontalBox_0
   widget=Overlay_Speaker1
   slot {class=HorizontalBoxSlot, properties {verticalAlignment=VAlign_Bottom, padding=<ltrb 0 0 6 0>}}
  }
  PlaceHolder {
   widgetClassPath=&Image
   parent=Overlay_Speaker1
   widget=PlaceHolder
   properties {
    brush {
     bIsDynamicallyLoaded=false
     drawAs=Image
     imageType=NoImage
     mirroring=NoMirror
     resourceName=None
     resourceObject=/Game/Athena/HUD/Bacchus/HUDLayoutTool/WidgetPreview/WidgetPreview_PlaceHolders/T-HLT-ActiveSpeaker_NoText
     tiling=NoTile
     imageSize=<265 55>
     margin=<ltrb 0 0 0 0>
     outlineSettings {
      bUseBrushTransparency=false
      roundingType=HalfHeightRadius
      width=0
      color {colorUseRule=UseColor_Specified, specifiedColor=<rgba 0 0 0 0>}
      cornerRadii=<0 0 0 0>
     }
     tintColor {colorUseRule=UseColor_Specified, specifiedColor=<rgba 1 1 1 1>}
     uVRegion {bIsValid=false, max=<0 0>, min=<0 0>}
    }
    renderTransform {angle=0, scale=<1 1>, shear=<0 0>, translation=<0 -2>}
   }
   slot {class=OverlaySlot, properties {horizontalAlignment=HAlign_Center, verticalAlignment=VAlign_Center}}
  }
  SizeBox_username {
   widgetClassPath=&SizeBox
   parent=Overlay_Speaker1
   widget=SizeBox_username
   properties {bOverride_HeightOverride=true, bOverride_WidthOverride=true, heightOverride=22, widthOverride=138}
   slot {class=OverlaySlot, properties {verticalAlignment=VAlign_Center, padding=<ltrb 70 0 0 13>}}
  }
  Text_DisplayName {
   widgetClassPath=&CommonTextBlock
   parent=SizeBox_username
   widget=Text_DisplayName
   properties {
    scrollStyle=&TextScrollStyle-Base_C
    text=User name
    textTransformPolicy=ToUpper
    colorAndOpacity {colorUseRule=UseColor_Specified, specifiedColor=<rgba 0.4452 1 1 1>}
    font {
     bForceMonospaced=false
     bMaterialIsStencil=false
     fontMaterial=null
     fontObject=&BurbankSmall
     letterSpacing=0
     monospacedWidth=1
     size=16
     skewAmount=0
     typefaceFontName=Black
     outlineSettings {
      bApplyOutlineToDropShadows=false
      bMiteredCorners=false
      bSeparateFillAlpha=false
      outlineMaterial=null
      outlineSize=0
      outlineColor=<rgba 0 0 0 1>
     }
    }
    shadowOffset=<0 0>
   }
   slot {class=SizeBoxSlot}
  }
  SizeBox_channelname {
   widgetClassPath=&SizeBox
   parent=Overlay_Speaker1
   widget=SizeBox_channelname
   properties {bOverride_HeightOverride=true, bOverride_WidthOverride=true, heightOverride=22, widthOverride=138}
   slot {class=OverlaySlot, properties {verticalAlignment=VAlign_Center, padding=<ltrb 70 11 0 0>}}
  }
  Text_DisplayName_4 {
   widgetClassPath=&CommonTextBlock
   parent=SizeBox_channelname
   widget=Text_DisplayName_4
   properties {
    scrollStyle=&TextScrollStyle-Base_C
    text=Channel name
    textTransformPolicy=ToUpper
    colorAndOpacity {colorUseRule=UseColor_Specified, specifiedColor=<rgba 0.4452 1 1 1>}
    font {
     bForceMonospaced=false
     bMaterialIsStencil=false
     fontMaterial=null
     fontObject=&BurbankSmall
     letterSpacing=0
     monospacedWidth=1
     size=12
     skewAmount=0
     typefaceFontName=Black
     outlineSettings {
      bApplyOutlineToDropShadows=false
      bMiteredCorners=false
      bSeparateFillAlpha=false
      outlineMaterial=null
      outlineSize=0
      outlineColor=<rgba 0 0 0 1>
     }
    }
    shadowOffset=<0 0>
   }
   slot {class=SizeBoxSlot}
  }
  Overlay_Speaker2 {
   widgetClassPath=&Overlay
   parent=HorizontalBox_0
   widget=Overlay_Speaker2
   slot {class=HorizontalBoxSlot, properties {verticalAlignment=VAlign_Bottom, padding=<ltrb 0 0 6 0>}}
  }
  PlaceHolder_1 {
   widgetClassPath=&Image
   parent=Overlay_Speaker2
   widget=PlaceHolder_1
   properties {
    brush {
     bIsDynamicallyLoaded=false
     drawAs=Image
     imageType=NoImage
     mirroring=NoMirror
     resourceName=None
     resourceObject=&T-HLT-ActiveSpeaker_NoText
     tiling=NoTile
     imageSize=<265 55>
     margin=<ltrb 0 0 0 0>
     outlineSettings {
      bUseBrushTransparency=false
      roundingType=HalfHeightRadius
      width=0
      color {colorUseRule=UseColor_Specified, specifiedColor=<rgba 0 0 0 0>}
      cornerRadii=<0 0 0 0>
     }
     tintColor {colorUseRule=UseColor_Specified, specifiedColor=<rgba 1 1 1 1>}
     uVRegion {bIsValid=false, max=<0 0>, min=<0 0>}
    }
    renderTransform {angle=0, scale=<1 1>, shear=<0 0>, translation=<0 -2>}
   }
   slot {class=OverlaySlot, properties {horizontalAlignment=HAlign_Center, verticalAlignment=VAlign_Center}}
  }
  SizeBox_username_1 {
   widgetClassPath=&SizeBox
   parent=Overlay_Speaker2
   widget=SizeBox_username_1
   properties {bOverride_HeightOverride=true, bOverride_WidthOverride=true, heightOverride=22, widthOverride=138}
   slot {class=OverlaySlot, properties {verticalAlignment=VAlign_Center, padding=<ltrb 70 0 0 13>}}
  }
  Text_DisplayName_1 {
   widgetClassPath=&CommonTextBlock
   parent=SizeBox_username_1
   widget=Text_DisplayName_1
   properties {
    scrollStyle=&TextScrollStyle-Base_C
    text=User name
    textTransformPolicy=ToUpper
    colorAndOpacity {colorUseRule=UseColor_Specified, specifiedColor=<rgba 0.4452 1 1 1>}
    font {
     bForceMonospaced=false
     bMaterialIsStencil=false
     fontMaterial=null
     fontObject=&BurbankSmall
     letterSpacing=0
     monospacedWidth=1
     size=16
     skewAmount=0
     typefaceFontName=Black
     outlineSettings {
      bApplyOutlineToDropShadows=false
      bMiteredCorners=false
      bSeparateFillAlpha=false
      outlineMaterial=null
      outlineSize=0
      outlineColor=<rgba 0 0 0 1>
     }
    }
    shadowOffset=<0 0>
   }
   slot {class=SizeBoxSlot}
  }
  SizeBox_channelname_1 {
   widgetClassPath=&SizeBox
   parent=Overlay_Speaker2
   widget=SizeBox_channelname_1
   properties {bOverride_HeightOverride=true, bOverride_WidthOverride=true, heightOverride=22, widthOverride=138}
   slot {class=OverlaySlot, properties {verticalAlignment=VAlign_Center, padding=<ltrb 70 11 0 0>}}
  }
  Text_DisplayName_2 {
   widgetClassPath=&CommonTextBlock
   parent=SizeBox_channelname_1
   widget=Text_DisplayName_2
   properties {
    scrollStyle=&TextScrollStyle-Base_C
    text=Channel name
    textTransformPolicy=ToUpper
    colorAndOpacity {colorUseRule=UseColor_Specified, specifiedColor=<rgba 0.4452 1 1 1>}
    font {
     bForceMonospaced=false
     bMaterialIsStencil=false
     fontMaterial=null
     fontObject=&BurbankSmall
     letterSpacing=0
     monospacedWidth=1
     size=12
     skewAmount=0
     typefaceFontName=Black
     outlineSettings {
      bApplyOutlineToDropShadows=false
      bMiteredCorners=false
      bSeparateFillAlpha=false
      outlineMaterial=null
      outlineSize=0
      outlineColor=<rgba 0 0 0 1>
     }
    }
    shadowOffset=<0 0>
   }
   slot {class=SizeBoxSlot}
  }
  Overlay_Speaker3 {
   widgetClassPath=&Overlay
   parent=HorizontalBox_0
   widget=Overlay_Speaker3
   slot {class=HorizontalBoxSlot, properties {verticalAlignment=VAlign_Bottom, padding=<ltrb 0 0 6 0>}}
  }
  PlaceHolder_2 {
   widgetClassPath=&Image
   parent=Overlay_Speaker3
   widget=PlaceHolder_2
   properties {
    brush {
     bIsDynamicallyLoaded=false
     drawAs=Image
     imageType=NoImage
     mirroring=NoMirror
     resourceName=None
     resourceObject=&T-HLT-ActiveSpeaker_NoText
     tiling=NoTile
     imageSize=<265 55>
     margin=<ltrb 0 0 0 0>
     outlineSettings {
      bUseBrushTransparency=false
      roundingType=HalfHeightRadius
      width=0
      color {colorUseRule=UseColor_Specified, specifiedColor=<rgba 0 0 0 0>}
      cornerRadii=<0 0 0 0>
     }
     tintColor {colorUseRule=UseColor_Specified, specifiedColor=<rgba 1 1 1 1>}
     uVRegion {bIsValid=false, max=<0 0>, min=<0 0>}
    }
    renderTransform {angle=0, scale=<1 1>, shear=<0 0>, translation=<0 -2>}
   }
   slot {class=OverlaySlot, properties {horizontalAlignment=HAlign_Center, verticalAlignment=VAlign_Center}}
  }
  SizeBox_username_2 {
   widgetClassPath=&SizeBox
   parent=Overlay_Speaker3
   widget=SizeBox_username_2
   properties {bOverride_HeightOverride=true, bOverride_WidthOverride=true, heightOverride=22, widthOverride=138}
   slot {class=OverlaySlot, properties {verticalAlignment=VAlign_Center, padding=<ltrb 70 0 0 13>}}
  }
  Text_DisplayName_3 {
   widgetClassPath=&CommonTextBlock
   parent=SizeBox_username_2
   widget=Text_DisplayName_3
   properties {
    scrollStyle=&TextScrollStyle-Base_C
    text=User name
    textTransformPolicy=ToUpper
    colorAndOpacity {colorUseRule=UseColor_Specified, specifiedColor=<rgba 0.4452 1 1 1>}
    font {
     bForceMonospaced=false
     bMaterialIsStencil=false
     fontMaterial=null
     fontObject=&BurbankSmall
     letterSpacing=0
     monospacedWidth=1
     size=16
     skewAmount=0
     typefaceFontName=Black
     outlineSettings {
      bApplyOutlineToDropShadows=false
      bMiteredCorners=false
      bSeparateFillAlpha=false
      outlineMaterial=null
      outlineSize=0
      outlineColor=<rgba 0 0 0 1>
     }
    }
    shadowOffset=<0 0>
   }
   slot {class=SizeBoxSlot}
  }
  SizeBox_channelname_2 {
   widgetClassPath=&SizeBox
   parent=Overlay_Speaker3
   widget=SizeBox_channelname_2
   properties {bOverride_HeightOverride=true, bOverride_WidthOverride=true, heightOverride=22, widthOverride=138}
   slot {class=OverlaySlot, properties {verticalAlignment=VAlign_Center, padding=<ltrb 70 11 0 0>}}
  }
  Text_DisplayName_5 {
   widgetClassPath=&CommonTextBlock
   parent=SizeBox_channelname_2
   widget=Text_DisplayName_5
   properties {
    scrollStyle=&TextScrollStyle-Base_C
    text=Channel name
    textTransformPolicy=ToUpper
    colorAndOpacity {colorUseRule=UseColor_Specified, specifiedColor=<rgba 0.4452 1 1 1>}
    font {
     bForceMonospaced=false
     bMaterialIsStencil=false
     fontMaterial=null
     fontObject=&BurbankSmall
     letterSpacing=0
     monospacedWidth=1
     size=12
     skewAmount=0
     typefaceFontName=Black
     outlineSettings {
      bApplyOutlineToDropShadows=false
      bMiteredCorners=false
      bSeparateFillAlpha=false
      outlineMaterial=null
      outlineSize=0
      outlineColor=<rgba 0 0 0 1>
     }
    }
    shadowOffset=<0 0>
   }
   slot {class=SizeBoxSlot}
  }
  Image_OutOfBoundsIndicator {
   widgetClassPath=&Image
   bInherited=true
   bIsVariable=true
   parent=Overlay_0
   widget=Image_OutOfBoundsIndicator
   properties {
    visibility=Collapsed
    brush {
     bIsDynamicallyLoaded=false
     drawAs=Image
     imageType=NoImage
     mirroring=NoMirror
     resourceName=None
     resourceObject=/Game/Athena/HUD/Bacchus/HUDLayoutTool/HLT_Icons/MI_HLT_OutOfBoundsIndicator
     tiling=NoTile
     imageSize=<32 32>
     margin=<ltrb 0 0 0 0>
     outlineSettings {
      bUseBrushTransparency=false
      roundingType=HalfHeightRadius
      width=0
      color {colorUseRule=UseColor_Specified, specifiedColor=<rgba 0 0 0 0>}
      cornerRadii=<0 0 0 0>
     }
     tintColor {colorUseRule=UseColor_Specified, specifiedColor=<rgba 1 1 1 1>}
     uVRegion {bIsValid=false, max=<0 0>, min=<0 0>}
    }
   }
   slot {class=OverlaySlot, properties {horizontalAlignment=HAlign_Fill, verticalAlignment=VAlign_Fill, padding=<ltrb -6 -4 -6 -6>}}
  }
  Image_SelectionHighlight {
   widgetClassPath=&Image
   bInherited=true
   parent=Overlay_0
   widget=Image_SelectionHighlight
   properties {
    visibility=Collapsed
    brush {
     bIsDynamicallyLoaded=false
     drawAs=Box
     imageType=NoImage
     mirroring=NoMirror
     resourceName=None
     resourceObject=/Game/Athena/HUD/Bacchus/HUDLayoutTool/HLT_Icons/M_UI_Square_Inst
     tiling=NoTile
     imageSize=<32 32>
     margin=<ltrb 0.25 0.25 0.25 0.25>
     outlineSettings {
      bUseBrushTransparency=false
      roundingType=HalfHeightRadius
      width=0
      color {colorUseRule=UseColor_Specified, specifiedColor=<rgba 0 0 0 0>}
      cornerRadii=<0 0 0 0>
     }
     tintColor {colorUseRule=UseColor_Specified, specifiedColor=<rgba 1 1 1 1>}
     uVRegion {bIsValid=false, max=<0 0>, min=<0 0>}
    }
    colorAndOpacity=<rgba 0.9301 1 0.0097 1>
   }
   slot {class=OverlaySlot, properties {horizontalAlignment=HAlign_Fill, verticalAlignment=VAlign_Fill, padding=<ltrb -8 -6 -8 -8>}}
  }
  Image_Background {widgetClassPath=&Image, bInherited=true, widget=null}
  TutorialHighlight {widgetClassPath=&TutorialHighlight}
 ]
}
