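# Widget Blueprint proxy that stands in for the FPS display in the mobile HUD Layout Tool, derived from FortMobileHUDElementProxy.
# A user-specified ScaleBox wraps a translucent rounded border and fixed-height SizeBox, whose overlay shows a static FPS placeholder texture.
# Inherited selection-highlight and out-of-bounds images overlay the placeholder and start collapsed.
# The asset has no graph logic or variables, so behavior comes from the parent class.
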WidgetBlueprint /Game/Athena/HUD/Bacchus/HUDLayoutTool/Proxies/HUDLayoutToolV2_FPSProxy
graphs [EventGraph {links [], nodes []}]
variables []
widgets {
 info {inheritedWidgetCount=5, parentClass=/Script/FortniteUI.FortMobileHUDElementProxy, rootWidgetClass=&ScaleBox, widgetCount=9}
 widgets [
  ScaleBox_HUDWidgetScale {widgetClassPath=&ScaleBox, bInherited=true, widget=ScaleBox_HUDWidgetScale, properties {stretch=UserSpecified}}
  Border_BG {
   widgetClassPath=&Border
   parent=ScaleBox_HUDWidgetScale
   widget=Border_BG
   properties {
    background {
     bIsDynamicallyLoaded=false
     drawAs=RoundedBox
     imageType=NoImage
     mirroring=NoMirror
     resourceName=None
     resourceObject=null
     tiling=NoTile
     imageSize=<32 32>
     margin=<ltrb 0 0 0 0>
     outlineSettings {
      bUseBrushTransparency=false
      roundingType=FixedRadius
      width=0
      color {colorUseRule=UseColor_Specified, specifiedColor=<rgba 0 0 0 0>}
      cornerRadii=<5 5 5 5>
     }
     tintColor {colorUseRule=UseColor_Specified, specifiedColor=<rgba 1 1 1 1>}
     uVRegion {bIsValid=false, max=<0 0>, min=<0 0>}
    }
    brushColor=<rgba 0 0 0 0.2>
    padding=<ltrb -15 -5 -15 -5>
   }
   slot {class=ScaleBoxSlot}
  }
  SizeBox_100 {
   widgetClassPath=&SizeBox
   parent=Border_BG
   widget=SizeBox_100
   properties {bOverride_HeightOverride=true, heightOverride=55}
   slot {class=BorderSlot, properties {padding=<ltrb -15 -5 -15 -5>}}
  }
  Overlay_4 {widgetClassPath=&Overlay, parent=SizeBox_100, widget=Overlay_4, slot {class=SizeBoxSlot}}
  PlaceHolder {
   widgetClassPath=&Image
   parent=Overlay_4
   widget=PlaceHolder
   properties {
    brush {
     bIsDynamicallyLoaded=false
     drawAs=Image
     imageType=NoImage
     mirroring=NoMirror
     resourceName=None
     resourceObject=/Game/Athena/HUD/Bacchus/HUDLayoutTool/WidgetPreview/WidgetPreview_PlaceHolders/T-HLT-FPS
     tiling=NoTile
     imageSize=<245 27>
     margin=<ltrb 0 0 0 0>
     outlineSettings {
      bUseBrushTransparency=false
      roundingType=HalfHeightRadius
      width=0
      color {colorUseRule=UseColor_Specified, specifiedColor=<rgba 0 0 0 0>}
      cornerRadii=<0 0 0 0>
     }
     tintColor {colorUseRule=UseColor_Specified, specifiedColor=<rgba 1 1 1 1>}
     uVRegion {bIsValid=false, max=<0 0>, min=<0 0>}
    }
    renderTransform {angle=0, scale=<0.9 0.9>, shear=<0 0>, translation=<0 -2>}
   }
   slot {class=OverlaySlot, properties {horizontalAlignment=HAlign_Center, verticalAlignment=VAlign_Center, padding=<ltrb 20 3 20 0>}}
  }
  Image_SelectionHighlight {
   widgetClassPath=&Image
   bInherited=true
   parent=Overlay_4
   widget=Image_SelectionHighlight
   properties {
    visibility=Collapsed
    brush {
     bIsDynamicallyLoaded=false
     drawAs=Box
     imageType=NoImage
     mirroring=NoMirror
     resourceName=None
     resourceObject=/Game/Athena/HUD/Bacchus/HUDLayoutTool/HLT_Icons/M_UI_Square_Inst
     tiling=NoTile
     imageSize=<32 32>
     margin=<ltrb 0.25 0.25 0.25 0.25>
     outlineSettings {
      bUseBrushTransparency=false
      roundingType=HalfHeightRadius
      width=0
      color {colorUseRule=UseColor_Specified, specifiedColor=<rgba 0 0 0 0>}
      cornerRadii=<0 0 0 0>
     }
     tintColor {colorUseRule=UseColor_Specified, specifiedColor=<rgba 1 1 1 1>}
     uVRegion {bIsValid=false, max=<0 0>, min=<0 0>}
    }
    colorAndOpacity=<rgba 0.9301 1 0.0097 1>
   }
   slot {class=OverlaySlot, properties {horizontalAlignment=HAlign_Fill, verticalAlignment=VAlign_Fill, padding=<ltrb 0 -10 0 -10>}}
  }
  Image_OutOfBoundsIndicator {
   widgetClassPath=&Image
   bInherited=true
   bIsVariable=true
   parent=Overlay_4
   widget=Image_OutOfBoundsIndicator
   properties {
    visibility=Collapsed
    brush {
     bIsDynamicallyLoaded=false
     drawAs=Image
     imageType=FullColor
     mirroring=NoMirror
     resourceName=None
     resourceObject=/Game/Athena/HUD/Bacchus/HUDLayoutTool/HLT_Icons/MI_HLT_BoldOutOfBoundsIndicator
     tiling=NoTile
     imageSize=<32 32>
     margin=<ltrb 0 0 0 0>
     outlineSettings {
      bUseBrushTransparency=false
      roundingType=HalfHeightRadius
      width=0
      color {colorUseRule=UseColor_Specified, specifiedColor=<rgba 0 0 0 0>}
      cornerRadii=<0 0 0 0>
     }
     tintColor {colorUseRule=UseColor_Specified, specifiedColor=<rgba 1 1 1 1>}
     uVRegion {bIsValid=false, max=<0 0>, min=<0 0>}
    }
   }
   slot {class=OverlaySlot, properties {horizontalAlignment=HAlign_Fill, verticalAlignment=VAlign_Fill, padding=<ltrb 0 -10 0 -10>}}
  }
  Image_Background {widgetClassPath=&Image, bInherited=true, widget=null}
  TutorialHighlight {widgetClassPath=&TutorialHighlight}
 ]
}
