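# Widget blueprint for a placeholder kill feed entry, used as a proxy in the Bacchus HUD Layout Tool V2.
# A fixed-width SizeBox wraps one RichTextBlock holding a hardcoded sample elimination message.
# The text uses the Bacchus kill feed style set, with tags for the local killer and victim, plus a kill feed image decorator.
# It has no graph logic or variables.
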WidgetBlueprint /Game/Athena/HUD/Bacchus/HUDLayoutTool/Proxies/HUDLayoutToolV2_KillFeedProxyEntry
graphs [EventGraph {links [], nodes []}]
variables []
widgets {
 info {inheritedWidgetCount=0, parentClass=&CommonUserWidget, rootWidgetClass=&SizeBox, widgetCount=2}
 widgets [
  SizeBox_347 {widgetClassPath=&SizeBox, widget=SizeBox_347, properties {bOverride_WidthOverride=true, widthOverride=700}}
  KillfeedText {
   widgetClassPath=/Script/UMG.RichTextBlock
   parent=SizeBox_347
   widget=KillfeedText
   properties {
    text="<LocalKiller>You</> eliminated <Victim>Epic Games</> with kindness"
    textStyleSet=/Game/Athena/HUD/Bacchus/BacchusKillFeedStyles
    decoratorClasses [/Game/Athena/HUD/TextStyles/KillFeedImageDecorator.KillFeedImageDecorator_C]
   }
   slot {class=SizeBoxSlot}
  }
 ]
}
