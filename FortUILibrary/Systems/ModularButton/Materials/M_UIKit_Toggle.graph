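# UI-domain translucent material for the UIKit modular button's toggle element, drawn procedurally as a radio button or checkbox.
# SDF circles and rounded boxes with pixel-sized outlines form idle, focused and selected masks, plus a texture-cut checkmark.
# ToggleType blends shapes through Lerp_3Color; its "Toggle" setting reuses the middle (checkbox) input.
# Button-state functions blend idle, focused, disabled and selected colors, with opacity from color alpha times rerouted masks.
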Material /FortUILibrary/Systems/ModularButton/Materials/M_UIKit_Toggle
graph {
 properties {
  bUsedWithStaticMesh=false
  blendMode=BLEND_Translucent
  materialDomain=MD_UI
  pixelDepthOffsetMode=PDOM_Legacy
  usedShadingModels=Default Lit
 }
 nodes [
  n1 = VectorParameter() {group=Colors, parameterName=ElementColor_Idle, sortPriority=0, defaultValue=<rgba 1 1 1 1>}
  n2 = VectorParameter() {group=Colors, parameterName=ElementColor_Disabled, sortPriority=4, defaultValue=<rgba 1 1 1 1>}
  n3 = MaterialFunctionCall!(
   Idle=@n1.RGB
   Idle Opacity=Multiply(A=NamedRerouteUsage(), B=@n1.A)
   Hovered
   Hovered Opacity
   Focused=@n1.RGB
   Focused Opacity=Multiply(A=NamedRerouteUsage(), B=@n1.A)
   Pressed
   Pressed Opacity
   Disabled=@n2.RGB
   Disabled Opacity=Multiply(A=NamedRerouteUsage(), B=@n2.A)
  ) {materialFunction=/FortUILibrary/Systems/ModularButton/MaterialFunctions/MF_UIKit_ButtonStatesBase_Lerp}
  n4 = VectorParameter() {group=Colors, parameterName=ElementColor_Selected, sortPriority=10, defaultValue=<rgba 1 1 1 1>}
  n5 = Multiply(A=@n4.A, B=NamedRerouteUsage())
  n6 = VectorParameter() {group=Colors, parameterName=ElementColor_Disabled, sortPriority=4, defaultValue=<rgba 1 1 1 1>}
  MaterialFunctionCall!(
   Base States=@n3.Color
   Base States Opacity=@n3.Color
   Selected_Idle=@n4.RGB
   Selected_Idle Opacity=@n5
   Selected_Hovered
   Selected_Hovered Opacity
   Selected_Focused=@n4.RGB
   Selected_Focused Opacity=@n5
   Selected_Pressed
   Selected_Pressed Opacity
   Selected_Disabled=@n6.RGB
   Selected_Disabled Opacity=Multiply(A=@n6.A, B=NamedRerouteUsage())
  ) {materialFunction=/FortUILibrary/Systems/ModularButton/MaterialFunctions/MF_UIKit_ButtonStatesSelectable_Lerp}
  n7 = ScalarParameter() {defaultValue=0.3, group=Colors, parameterName=BackplateAlpha, sliderMax=1, sortPriority=30}
  n8 = MaterialFunctionCall(Size=Constant() {r=1}) {materialFunction=&MF_UI_Circle}
  n9 = MaterialFunctionCall(DistanceField=@n8.SDF) {materialFunction=&MF_UI_AntiAliasing}
  n10 = ScalarParameter() {defaultValue=2, group=Layout, parameterName=Outline_PX}
  n11 = NamedRerouteDeclaration(None=MaterialFunctionCall!(Value PX=Multiply(A=@n10) {constB=1.5}, UV Pixel Size, X or Y) {materialFunction=/FortUILibrary/MaterialFunctions/Basics/MF_UI_PixelsToNormalized}."Normalized (>1px)") {name=Edge Outline Size, nodeColor=<rgba 0 0 0 1>}
  n12 = Add(A=@n8.SDF, B=@n11)
  n13 = MaterialFunctionCall(DistanceField=Subtract(A=Abs(None=@n12), B=@n11) {constB=0.1}) {materialFunction=&MF_UI_AntiAliasing}
  NamedRerouteDeclaration(None=LinearInterpolate(A=Multiply(A=@n7, B=@n9.Inner), Alpha=@n13.Inner)) {name=Radio_Idle, nodeColor=<rgba 0 0 0 1>}
  n14 = NamedRerouteUsage()
  n15 = ScalarParameter() {defaultValue=5, group=Layout, parameterName=EdgeOffset_InnerRadio_PX}
  n16 = Saturate(None=Add(
    A=Multiply(A=@n14) {constB=1.5}
    B=MaterialFunctionCall(Value PX=@n15) {materialFunction=&MF_UI_PixelsToNormalized}."Normalized (>1px)"
   ) {constB=0})
  n17 = MaterialFunctionCall(Size=NamedRerouteDeclaration(None=OneMinus(None=@n16)) {name=Offset Inner Outline Size, nodeColor=<rgba 0 0 0 1>}) {materialFunction=&MF_UI_Circle}
  n18 = NamedRerouteDeclaration(None=MaterialFunctionCall(Value PX=ScalarParameter() {defaultValue=5, group=Layout, parameterName=Outline_InnerRadio_PX}) {materialFunction=&MF_UI_PixelsToNormalized}."Normalized (>1px)") {name=Inner Outline Size, nodeColor=<rgba 0 0 0 1>}
  n19 = Subtract(A=Abs(None=Add(A=@n17.SDF, B=@n18)), B=@n18) {constB=0.1}
  NamedRerouteDeclaration(None=LinearInterpolate(
    A=@n13.Inner
    Alpha=MaterialFunctionCall(DistanceField=@n19) {materialFunction=&MF_UI_AntiAliasing}.Inner
   )) {name=Radio_Selected, nodeColor=<rgba 0 0 0 1>}
  n20 = ScalarParameter() {defaultValue=0.3, group=Colors, parameterName=BackplateAlpha, sliderMax=1, sortPriority=30}
  n21 = MaterialFunctionCall(Value PX=Multiply(A=Constant() {r=8}) {constB=2}) {materialFunction=&MF_UI_PixelsToNormalized}
  n22 = StaticBool() {value=true}
  n23 = MaterialFunctionCall(Size=Constant() {r=1}, CornerRoundness=@n21."Normalized (>1px)", KeepAspectRatio=@n22) {materialFunction=&MF_UI_Box}
  n24 = MaterialFunctionCall(DistanceField=@n23.SDF) {materialFunction=&MF_UI_AntiAliasing}
  n25 = NamedRerouteUsage()
  n26 = Add(A=@n23.SDF, B=@n25)
  n27 = MaterialFunctionCall(DistanceField=Subtract(A=Abs(None=@n26), B=@n25) {constB=0.1}) {materialFunction=&MF_UI_AntiAliasing}
  NamedRerouteDeclaration(None=LinearInterpolate(A=Multiply(A=@n20, B=@n24.Inner), Alpha=@n27.Inner)) {name=Check_Idle, nodeColor=<rgba 0 0 0 1>}
  n28 = MaterialFunctionCall(DistanceField=MaterialFunctionCall(Size=NamedRerouteUsage(), CornerRoundness=Multiply(A=@n21."Normalized (>1px)") {constB=0.5}, KeepAspectRatio=@n22) {materialFunction=&MF_UI_Box}.SDF) {materialFunction=&MF_UI_AntiAliasing}
  n29 = TextureSample(UVs=MaterialFunctionCall(
    UVs=MaterialFunctionCall() {materialFunction=&GetUserInterfaceUV}.Normalized UV
    Texture Scale=Constant() {r=0.5}
   ) {materialFunction=&ScaleUVsByCenter}.UVs) {samplerSource=SSM_Clamp_WorldGroupSettings, samplerType=SAMPLERTYPE_Alpha, texture=/FortUILibrary/Textures/Icons/T_UI_Icon_UIKit_Check}
  NamedRerouteDeclaration(None=LinearInterpolate(B=LinearInterpolate(A=@n27.Inner, Alpha=@n28.Inner), Alpha=OneMinus(None=@n29.R))) {name=Check_Selected, nodeColor=<rgba 0 0 0 1>}
  n30 = ScalarParameter() {defaultValue=0.6, group=Colors, parameterName=BackplateFocusedAlpha, sliderMax=1, sortPriority=31}
  NamedRerouteDeclaration(None=LinearInterpolate(A=Multiply(A=@n30, B=@n9.Inner), Alpha=@n13.Inner)) {name=Radio_Focused, nodeColor=<rgba 0 0 0 1>}
  n31 = ScalarParameter() {defaultValue=0.6, group=Colors, parameterName=BackplateFocusedAlpha, sliderMax=1, sortPriority=31}
  NamedRerouteDeclaration(None=LinearInterpolate(A=Multiply(A=@n31, B=@n24.Inner), Alpha=@n27.Inner)) {name=Check_Focused, nodeColor=<rgba 0 0 0 1>}
  n32 = NamedRerouteUsage()
  n33 = ScalarParameter() {defaultValue=1, desc="0: Radio \r\n0,5: CheckBox\r\n1: Toggle", group=Layout, parameterName=ToggleType, sliderMax=1, sortPriority=0}
  NamedRerouteDeclaration(None=MaterialFunctionCall(A=NamedRerouteUsage(), B=@n32, C=@n32, Alpha=@n33) {materialFunction=&Lerp_3Color}.Result) {name=Element_Idle, nodeColor=<rgba 0 0 0 1>}
  n34 = NamedRerouteUsage()
  NamedRerouteDeclaration(None=MaterialFunctionCall(
    A=NamedRerouteUsage()
    B=@n34
    C=@n34
    Alpha=ScalarParameter() {defaultValue=1, desc="0: Radio \r\n0,5: CheckBox\r\n1: Toggle", group=Layout, parameterName=ToggleType, sortPriority=0}
   ) {materialFunction=&Lerp_3Color}.Result) {name=Element_Selected, nodeColor=<rgba 0 0 0 1>}
  n35 = NamedRerouteUsage()
  NamedRerouteDeclaration(None=MaterialFunctionCall(
    A=NamedRerouteUsage()
    B=@n35
    C=@n35
    Alpha=ScalarParameter() {defaultValue=1, desc="0: Radio \r\n0,5: CheckBox\r\n1: Toggle", group=Layout, parameterName=ToggleType, sortPriority=0}
   ) {materialFunction=&Lerp_3Color}.Result) {name=Element_Focused, nodeColor=<rgba 0 0 0 1>}
 ]
}
