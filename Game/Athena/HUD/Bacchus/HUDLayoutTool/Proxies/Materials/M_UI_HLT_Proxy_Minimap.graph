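# Translucent UI material that draws a procedural stand-in minimap for the HUD Layout Tool.
# UV-scale masks form the outline and faint background, while three SDF circles build stacked faint discs offset from center.
# Three rotated UV half-plane gradients combine through saturated dot products into a player arrow.
# A FinalOpacity reroute sums four layer reroutes, alongside a Color parameter that defaults to white.
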Material /Game/Athena/HUD/Bacchus/HUDLayoutTool/Proxies/Materials/M_UI_HLT_Proxy_Minimap
graph {
 properties {bAutomaticallySetUsageInEditor=false, blendMode=BLEND_Translucent, materialDomain=MD_UI, usedShadingModels=Default Lit}
 nodes [
  n1 = TextureCoordinate()
  n2 = OneMinus(None=Constant() {r=0.025})
  NamedRerouteDeclaration(None=OneMinus(None=MaterialFunctionCall(UVs=@n1, Texture Scale=@n2) {materialFunction=&ScaleUVsByCenter}.0-1 mask)) {name=Outline, nodeColor=<rgba 0 0 0 1>}
  n3 = TextureCoordinate()
  NamedRerouteDeclaration(None=Multiply(A=MaterialFunctionCall(UVs=@n3) {materialFunction=&ScaleUVsByCenter}.0-1 mask) {constB=0.2, desc=Opacity BG}) {name=BG, nodeColor=<rgba 0 0 0 1>}
  n4 = Add(A=NamedRerouteUsage(), B=NamedRerouteUsage())
  n5 = NamedRerouteUsage()
  NamedRerouteDeclaration(None=Add(A=Add(A=@n4, B=@n5), B=NamedRerouteUsage())) {name=FinalOpacity, nodeColor=<rgba 0 0 0 1>}
  n6 = Add(
   A=MaterialFunctionCall(UVs=TextureCoordinate(), Texture Scale=Constant() {r=3}) {materialFunction=&ScaleUVsByCenter}.UVs
   B=Constant2Vector() {g=-0.1, r=-0.1}
  )
  n7 = OneMinus(None=MaterialFunctionCall!(UVs=@n6, Radius) {materialFunction=/FortUILibrary/MaterialFunctions/Shapes/SDFs/MF_SDF_Circle}.SDF)
  n8 = Multiply(A=Clamp(None=Floor(None=@n7))) {constB=0.04}
  n9 = OneMinus(None=MaterialFunctionCall(UVs=@n6, Radius=Constant() {r=0.4}) {materialFunction=&MF_SDF_Circle}.SDF)
  n10 = Multiply(A=Clamp(None=Floor(None=@n9))) {constB=0.04}
  n11 = MaterialFunctionCall(UVs=@n6, Radius=Constant() {r=0.3}) {materialFunction=&MF_SDF_Circle}
  n12 = Clamp(None=Floor(None=OneMinus(None=@n11.SDF)))
  NamedRerouteDeclaration(None=Add(A=Add(A=@n8, B=@n10), B=Multiply(A=@n12) {constB=0.04})) {name=Circles, nodeColor=<rgba 0 0 0 1>}
  n13 = MaterialFunctionCall(
   UVs=MaterialFunctionCall(UVs=TextureCoordinate(), Texture Scale=Constant() {r=0.15}) {materialFunction=&ScaleUVsByCenter}.UVs
   "Rotation Angle (0-1)"=Constant() {r=0.1}
  ) {materialFunction=&CustomRotator}
  n14 = Add(
   A=MaterialFunctionCall(UVs=@n13.Rotated Values, "Rotation Angle (0-1)"=Constant() {r=0.07}) {materialFunction=&CustomRotator}.Rotated Values
   B=Multiply() {constA=0.25, constB=-1}
  ) {constB=-0.5}
  n15 = Add(
   A=MaterialFunctionCall(UVs=@n13.Rotated Values, "Rotation Angle (0-1)"=Constant() {r=0.3746}) {materialFunction=&CustomRotator}.Rotated Values
   B=Multiply() {constA=0.25, constB=-1}
  ) {constB=-0.5}
  n16 = MaterialFunctionCall(UVs=@n13.Rotated Values, "Rotation Angle (0-1)"=Constant() {r=0.6667}) {materialFunction=&CustomRotator}
  n17 = Multiply() {constA=0.25, constB=-1}
  n18 = AppendVector(
   A=AppendVector(A=ComponentMask(None=@n14) {g=true}, B=ComponentMask(None=@n15) {g=true})
   B=ComponentMask(None=Add(A=@n16.Rotated Values, B=@n17) {constB=-0.5}) {g=true}
  )
  n19 = MaterialFunctionCall(Float3=Saturate(None=Multiply(A=@n18) {constB=32})) {materialFunction=&BreakOutFloat3Components}
  n20 = DotProduct(A=@n19.R, B=@n19.R)
  NamedRerouteDeclaration(None=Saturate(None=DotProduct(A=@n20, B=@n19.B))) {name=PlayerArrow, nodeColor=<rgba 0 0 0 1>}
  NamedRerouteUsage()
  VectorParameter() {parameterName=Color, defaultValue=<rgba 1 1 1 1>}
 ]
}
