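# Translucent UI material for a progress bar proxy in the HUD Layout Tool, built as a static striped pattern.
# It slightly rotates scaled UVs from the second coordinate channel, then takes the inverted fractional X to form a repeating ramp.
# An add and floor turn the ramp into hard-edged bands, clamped so the dark bands keep a faint value, alongside a constant white color.
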Material /Game/Athena/HUD/Bacchus/HUDLayoutTool/Proxies/Materials/M_UI_HLT_Proxy_ProgressBar
graph {
 properties {bAutomaticallySetUsageInEditor=false, blendMode=BLEND_Translucent, materialDomain=MD_UI, usedShadingModels=Default Lit}
 nodes [
  n1 = MaterialFunctionCall(
   UVs=TextureCoordinate() {coordinateIndex=1, uTiling=0.6, vTiling=0.6}
   Rotation Center=Constant()
   "Rotation Angle (0-1)"=Constant() {r=0.99}
  ) {materialFunction=&CustomRotator}
  n2 = OneMinus(None=ComponentMask(None=Frac(None=@n1.Rotated Values)) {r=true})
  Clamp(None=Floor(None=Add(A=@n2) {constB=0.4})) {minDefault=0.15}
  Constant3Vector() {constant=<rgba 1 1 1 0>}
 ]
}
